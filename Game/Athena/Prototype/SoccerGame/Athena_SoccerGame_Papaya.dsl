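; Papaya variant of the Athena soccer game actor that respawns a soccer ball left out of bounds.
; On the authority, a ball ending overlap with OutOfBoundsVolume is stored and starts a delayed reset timer, which re-entering the volume clears.
; When the timer fires, it calls RespawnBall and silently kills the stored ball if still valid.
; The construction script calls the parent and places the volume from a transform variable.
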(event OnComponentBeginOverlap(OutOfBoundsVolume)
  (params (OverlappedComponent :type Primitive%20Component%20Object%20Reference) (OtherActor :type Actor%20Object%20Reference) (OtherComp :type Primitive%20Component%20Object%20Reference) (OtherBodyIndex :type Integer) (bFromSweep :type Boolean) (SweepResult :type Hit%20Result%20Structure%20%28by%20ref%29))
  ;[1 1 1] Clear the reset timer when the ball goes back in bounds
  (Networking|SwitchHasAuthority
    (:Authority
      (Utilities|Casting|CastToAthena_Soccerball OtherActor)
      (Utilities|Time|ClearandInvalidateTimerbyHandle (Variables|Default|GetTimer_RespawnBall))
      (Development|PrintString "Soccer Ball In Bounds" false false))))

(event OnComponentEndOverlap(OutOfBoundsVolume)
  (params (OverlappedComponent :type Primitive%20Component%20Object%20Reference) (OtherActor :type Actor%20Object%20Reference) (OtherComp :type Primitive%20Component%20Object%20Reference) (OtherBodyIndex :type Integer))
  ;[1 1 1] Respawn the ball if it has been out of bounds too long
  (Networking|SwitchHasAuthority
    (:Authority
      (bind _asathena_soccerball (Utilities|Casting|CastToAthena_Soccerball OtherActor))
      (Variables|Default|SetSoccerBall _asathena_soccerball)
      (bind _returnvalue (Utilities|Time|SetTimerbyEvent (delegate Custom|OutOfBoundsReset) (Variables|Default|GetOutOfBoundsRespawnBallDelay)))
      (Variables|Default|SetTimer_RespawnBall _returnvalue)
      (Development|PrintString "Soccer Ball Out Of Bounds" false false))))

(event Custom|OutOfBoundsReset
  (bind _soccerball (Variables|Default|GetSoccerBall))
  ;[1 1 1] Respawn the ball if it has been out of bounds too long
  (CallFunction|RespawnBall)
  (Utilities|IsValid _soccerball
    (:Is%20Valid
      (Gameplay|SilentDie _soccerball false :!class /Script/FortniteGame.BuildingActor))))

(fn ConstructionScript
  (|CallParentFunction)
  (Transformation|SetRelativeTransform (Variables|Default|GetOutOfBoundsVolume) (Variables|Default|GetOutOfBoundsVolumeTransform) false false))
